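# M_VoronoiHatchingLines is a stylized material that shades surfaces with hatching lines broken up by a Voronoi map.
# The map supplies cell ID, edge and center-distance masks, and the UV lines take per-region scale and rotation offsets.
# A Custom node reads directional light direction for biased smoothstep shading, overlay blending and a view-dependent backlight rim.
# Base, shaded and shadow colors blend by mask, while material attributes zero Base Color and drive Emissive.
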Material /Game/Global/Materials/MaterialGym/Materials/VoronoiHatchingLines/M_VoronoiHatchingLines
graph {
 properties {bUseMaterialAttributes=true, pixelDepthOffsetMode=PDOM_Legacy, usedShadingModels=Default Lit}
 nodes [
  n1 = VectorParameter() {group=01_VoronoiHatchingLines, parameterName=ShadowColor, sortPriority=3, defaultValue=<rgba 0 0 0 1>}
  n2 = VectorParameter() {group=01_VoronoiHatchingLines, parameterName=BaseColor, sortPriority=3, defaultValue=<rgba 0.9 0.9 0.9 1>}
  n3 = Multiply(A=ScalarParameter() {defaultValue=0.25, group=01_VoronoiHatchingLines, parameterName=ShadedBaseColor_Intensity, sliderMax=1}, B=@n2.RGB)
  n4 = NamedRerouteUsage()
  NamedRerouteDeclaration(None=LinearInterpolate(A=@n1.RGB, B=LinearInterpolate(A=@n3, B=@n2.RGB, Alpha=@n4), Alpha=NamedRerouteUsage())) {name=BaseColor_Raw, nodeColor=<rgba 1 0.4506 0.2387 1>}
  n5 = NamedRerouteUsage()
  n6 = ScalarParameter() {defaultValue=3, group=01_VoronoiHatchingLines, parameterName=BackLit_Exponent, sliderMax=8}
  n7 = Saturate(None=SmoothStep(Value=Power(Base=@n5, Exp=@n6) {constExponent=4}) {constMax=0, constMin=0.02})
  n8 = MaterialFunctionCall() {materialFunction=&CameraDirectionVector}
  n9 = NamedRerouteUsage()
  n10 = Clamp(
   None=Add(
    A=DotProduct(A=@n8.Result, B=@n9)
    B=ScalarParameter() {defaultValue=0.25, group=01_VoronoiHatchingLines, parameterName=BackLit_Bias, sliderMax=1}
   )
   Max=ScalarParameter() {defaultValue=0.75, group=01_VoronoiHatchingLines, parameterName=BackLit_MaxBias, sliderMax=1}
  )
  NamedRerouteDeclaration(None=Multiply(
    A=NamedRerouteDeclaration(None=@n7) {name=RimLight_Mask, nodeColor=<rgba 0 1 0.5078 1>}
    B=NamedRerouteDeclaration(None=@n10) {name=BackLit_Curve, nodeColor=<rgba 0.9507 0.45 1 1>}
   )) {name=BackLitShading_Curve, nodeColor=<rgba 0.1307 1 0.334 1>}
  n11 = Multiply(
   A=NamedRerouteUsage()
   B=ScalarParameter() {defaultValue=0.5, group=01_VoronoiHatchingLines, parameterName=LineDistortion_RandScale, sliderMax=1}
  )
  n12 = NamedRerouteUsage()
  n13 = ScalarParameter() {defaultValue=0.5, group=01_VoronoiHatchingLines, parameterName=LineDistortion_RandRot, sliderMax=0.25}
  n14 = Rotator(
   Coordinate=Multiply(A=TextureCoordinate(), B=Add(A=@n11))
   Time=Add(A=ScalarParameter() {group=01_VoronoiHatchingLines, parameterName=Lines_RotationOffset, sliderMax=1}, B=Multiply(A=@n12, B=@n13))
  ) {centerX=0, centerY=0, speed=-6.283}
  n15 = Frac(None=Multiply(A=ComponentMask(None=@n14) {r=true}, B=ScalarParameter() {defaultValue=180, group=01_VoronoiHatchingLines, parameterName=UV_Lines_Amount}) {constB=32})
  n16 = Abs(None=Add(A=Multiply(A=@n15) {constB=2}) {constB=-1}) {desc=HatchingDF}
  n17 = NamedRerouteUsage()
  NamedRerouteDeclaration(None=Saturate(None=LinearInterpolate(B=@n16, Alpha=@n17) {constA=1})) {name=UVHatching, nodeColor=<rgba 0.9922 1 0 1>}
  n18 = MaterialFunctionCall(Float2=TextureCoordinate()) {materialFunction=&BreakOutFloat2Components}
  n19 = Multiply(A=@n18.G, B=Divide(B=ScalarParameter() {defaultValue=1, group=01_VoronoiHatchingLines, parameterName=ShadowVariation_Ratio, sliderMax=2}))
  n20 = TextureSampleParameter2D(UVs=Multiply(
    A=MaterialFunctionCall(X=@n18.R, Y=@n19) {materialFunction=&MakeFloat2}.Result
    B=ScalarParameter() {defaultValue=4, group=01_VoronoiHatchingLines, parameterName=ShadowVariation_Tiling, sliderMax=16}
   )) {
   group=01_VoronoiHatchingLines
   parameterName=VoronoiMap
   samplerType=SAMPLERTYPE_LinearColor
   sortPriority=2
   texture=/Game/Global/Materials/Character_Parent_2023/Anime_Parent_2023/Textures/T_Hatching_Voronoi
  }
  NamedRerouteDeclaration(None=@n20.G) {name=VoronoiID, nodeColor=<rgba 1 0.2603 0.2703 1>}
  n21 = NamedRerouteUsage()
  n22 = Add(A=@n21, B=CameraVectorWS())
  NamedRerouteDeclaration(None=DotProduct(A=Normalize(VectorInput=@n22), B=PixelNormalWS())) {name=NdotH, nodeColor=<rgba 0.7344 1 0 1>}
  n23 = Constant() {r=0.5}
  n24 = Multiply(A=Saturate(None=ScalarParameter() {defaultValue=0.5, group=01_VoronoiHatchingLines, parameterName=ShadedBaseColor_TransitionAmount, sliderMax=1})) {constB=0.5}
  n25 = NamedRerouteUsage()
  n26 = ScalarParameter() {group=01_VoronoiHatchingLines, parameterName=ShadedBaseColor_Bias, sliderMax=1}
  n27 = Add(A=Multiply(A=Subtract(A=@n25, B=@n26)) {constB=0.5}) {constB=0.5}
  n28 = NamedRerouteUsage()
  NamedRerouteDeclaration(None=SmoothStep(Min=Subtract(A=@n23, B=@n24) {constA=0.5}, Max=Add(A=@n23, B=@n24) {constA=0.5, constB=-0.5}, Value=LinearInterpolate(A=@n27, Alpha=@n28)) {constMax=0.01}) {name=ShadedBaseColor_Curve, nodeColor=<rgba 0 0.9439 1 1>}
  NamedRerouteDeclaration(None=Custom() {code="return View.DirectionalLightDirection;", description=DirectionalLightDirection, inputs []}) {name=LightVector, nodeColor=<rgba 0.4766 1 0 1>}
  NamedRerouteDeclaration(None=DotProduct(A=CameraVectorWS(), B=PixelNormalWS())) {name=NdotV, nodeColor=<rgba 0.7344 1 0 1>}
  n29 = Constant()
  n30 = SetMaterialAttributes!(
   MaterialAttributes
   Base Color=@n29
   Specular=@n29
   Metallic=NamedRerouteUsage()
   Roughness
   Emissive Color=NamedRerouteUsage()
   Shading Model=ShadingModel() {shadingModel=MSM_Cloth}
  ) {attributeSetTypes [
    #1
    #2
    #3
    #4
    #5
    #6
   ]}
  NamedRerouteDeclaration(None=@n30) {name=MA_Out, nodeColor=<rgba 0 0.4375 1 1>}
  n31 = Constant() {r=0.5}
  n32 = Multiply(A=Saturate(None=ScalarParameter() {defaultValue=0.5, group=01_VoronoiHatchingLines, parameterName=Shadow_TransitionAmount, sliderMax=1})) {constB=0.5}
  n33 = Multiply(A=Subtract(A=NamedRerouteUsage(), B=ScalarParameter() {group=01_VoronoiHatchingLines, parameterName=Shadow_Bias, sliderMax=1, sliderMin=-1})) {constB=0.5}
  n34 = NamedRerouteUsage()
  n35 = ScalarParameter() {defaultValue=0.2, group=01_VoronoiHatchingLines, parameterName=ShadowVariation_Intensity, sliderMax=1}
  n36 = Add(A=Multiply(A=Add(A=@n33) {constB=0.5}, B=Constant() {r=2}), B=Subtract(A=Multiply(A=@n34, B=@n35) {constB=0.2}, B=Constant() {r=1}))
  n37 = SmoothStep(
   Min=Subtract(A=@n31, B=@n32) {constA=0.5}
   Max=Add(A=@n31, B=@n32) {constA=0.5, constB=-0.5}
   Value=MaterialFunctionCall(Base=NamedRerouteUsage(), Blend=Saturate(None=@n36) {desc=height blend with 1 heighmap}) {materialFunction=&Blend_Overlay}.Result
  ) {constMax=0.01}
  n38 = NamedRerouteUsage()
  NamedRerouteDeclaration(None=Saturate(None=LinearInterpolate(A=@n37, Alpha=@n38))) {name=ShadingMask, nodeColor=<rgba 1 0.6562 0 1>}
  NamedRerouteDeclaration(None=SmoothStep(Value=@n20.R) {constMax=0.8, constMin=1}) {name=VoronoiIEdgeDistance, nodeColor=<rgba 1 0.2603 0.2703 1>}
  NamedRerouteDeclaration(None=@n20.B) {name=VoronoiICenterDistance, nodeColor=<rgba 1 0.2603 0.2703 1>}
  NamedRerouteDeclaration(None=DotProduct(A=@n21, B=PixelNormalWS())) {name=NdotL, nodeColor=<rgba 0.7344 1 0 1>}
  NamedRerouteDeclaration(None=Constant() {r=0.5}) {name=Threshold, nodeColor=<rgba 0 0.5 1 1>}
 ]
}
